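# Editor utility widget for outsource environment review that lays out filtered sublevels or collection assets in the open map with text labels.
# Asset Registry and Asset Tags queries fill a shared list, switchable by workflow and filtered by text or season.
# Sublevels load through spawned 1x1 foundations with editor preview enabled, while each collection becomes a column of actors.
# Clearing destroys all foundations, text and building actors, including pre-existing ones.
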EditorUtilityWidgetBlueprint /Game/Environments/Blueprints/OS_Review/EUW_OSReview_Setup
graphs [
 DeleteMe_SpawnCollectionStuff {nodes [
   DeleteMe_SpawnCollectionStuff() {is=title K2Node_FunctionEntry}
   n1 = Get DeleteMe_CollectionName() -> (DeleteMe_CollectionName:Name=None)
   > n2 = GetAssetsInCollection!(
    execute
    self:Asset Tags Subsystem Object Reference=Get AssetTagsSubsystem() -> (ReturnValue:Asset Tags Subsystem Object Reference) {is=title K2Node_GetEngineSubsystem}
    Name:Name=@n1
   ) -> (then, ReturnValue:Array of Asset Data Structures)
   > n3 = For Each Loop(Array:Array of Asset Data Structures=@n2) -> (Array Element:Asset Data Structure)
   > PrintString(InString="To String (Integer)"(InInt=@n3.Array Index), Duration=10)
   n4 = float * float(A:int=Get SpawnCount() -> (SpawnCount:int=0), B:int=1024) {is=title K2Node_PromotableOperator}
   n5 = BreakVector(InVec=Get StartLocation() -> (StartLocation:Vector))
   n6 = float * float(A=@n3.Array Index, B:double=1024)
   > SpawnActorFromObject!(
    execute
    ObjectToUse:Object Reference=GetAsset(InAssetData=@n3)
    Location:Vector=MakeVector(X=float - float(A=@n5, B=@n4), Y=float - float(A=@n5.Y, B=@n6))
    Rotation:Rotator=Get Collection_StartRotation() -> (Collection_StartRotation:Rotator)
    bTransient:bool=false
   ) -> (then, ReturnValue:Actor Object Reference)
   n7 = BreakVector(InVec=Get StartLocation())
   n8 = float - float(A=@n7, B=float * float(A=Get SpawnCount() -> (SpawnCount=0), B=1024))
   n9 = BreakRotator(InRot=Get StartRotation() -> (StartRotation:Rotator))
   SpawnTextActor_FromCollection!(
    execute=@n3.Completed
    self
    Location:Vector=MakeVector(X=float - float(A=@n8, B=512), Y=@n7.Y, Z=1)
    SourceString:String="To String (Name)"(InName=@n1)
    Rotation:Rotator=MakeRotator(Roll=@n9, Pitch=60, Yaw=@n9.Yaw)
   ) -> (then)
   > Set SpawnCount(SpawnCount:int=int + int(A=Get SpawnCount() -> (SpawnCount=0))) -> (Output_Get:int=0)
  ]}
 EventGraph {nodes [
   n1 = getSelectedData!(execute, self:TA Generic 01 List Object Reference=Get TL_FilteredList() -> (TL_FilteredList:TA Generic 01 List Object Reference)) -> (then, selectedData:Array of Object References)
   > SpawnAssetsFromCollection!(execute, self, Array:Array of Strings=@n1) -> (then)
   n2 = "On Text Committed (TB_TextFilter)"!() -> (OutputDelegate, then, Text:"Text (by ref)", CommitMethod:ETextCommit Enum=Default) {is=title K2Node_ComponentBoundEvent}
   n3 = Get WS_WorkflowSwap() -> (WS_WorkflowSwap:Widget Switcher Object Reference)
   > n4 = Branch(Condition="Equal (Integer)"(A=GetActiveWidgetIndex!(self:Widget Switcher Object Reference=@n3) -> (ReturnValue:int=0)))
   > clearItems(self:TA Generic 01 List Object Reference=Get TL_FilteredList())
   > Set SearchFilter(SearchFilter:String="To String (Text)"(InText=@n2.Text)) -> (Output_Get:String)
   > GetFilteredSublevels()
   > PopulateListfromArray!(execute, self, Array:Array of Names=Get FilteredLevelsList() -> (FilteredLevelsList:Array of Names)) -> (then)
   > ClearSelection(self:"Combo Box (String) Object Reference"=Get DD_SeasonSelector() -> (DD_SeasonSelector:"Combo Box (String) Object Reference"))
   clearItems(execute=@n4.else, self=Get TL_FilteredList())
   > Set SearchFilter(SearchFilter="To String (Text)"(InText=@n2.Text))
   > GetFilteredCollections()
   > PopulateListfromCollections!(execute, self, Array:Array of Names=Get FilteredCollectionList() -> (FilteredCollectionList:Array of Names)) -> (then)
   "On Clicked (BTN_SpawnAllLevels)"!() -> (OutputDelegate, then, widget:User Widget Object Reference, event:Pointer Event Structure) {is=title K2Node_ComponentBoundEvent}
   > n5 = getListData!(execute, self:TA Generic 01 List Object Reference=Get TL_FilteredList()) -> (then, data:Array of Object References)
   > n6 = Branch(Condition="integer >= integer"(A=Length(TargetArray:Array of Object References=@n5), B=20))
   > n7 = Show Message Dialog(
    Title=Get ModalPopupTitle() -> (ModalPopupTitle:Text)
    Message=Get ModalPopupMessageText() -> (ModalPopupMessageText:Text)
    MessageType=YesNo
    DefaultValue=No
    MessageCategory=Warning
   )
   > n8 = Switch on EAppReturnType(Selection=@n7)
   n9 = getListData(execute=[@n8.Yes, @n6.else], self=Get TL_FilteredList())
   > SpawnLevelsFromListData!(execute, self, Array:Array of Object References=@n9) -> (then)
   "On Clicked (BTN_ClearLevels)"!() -> (OutputDelegate, then, widget:User Widget Object Reference, event:Pointer Event Structure) {is=title K2Node_ComponentBoundEvent}
   > RemoveAllAssets()
   "On Clicked (BTN_SpawnSelectedLevels)"!() -> (OutputDelegate, then, widget:User Widget Object Reference, event:Pointer Event Structure) {is=title K2Node_ComponentBoundEvent}
   > n10 = getSelectedData(self=Get TL_FilteredList())
   > n11 = Branch(Condition="integer >= integer"(A=Length(TargetArray=@n10), B=20))
   > n12 = Show Message Dialog(Title=Get ModalPopupTitle(), Message=Get ModalPopupMessageText(), MessageType=YesNo, DefaultValue=No, MessageCategory=Warning)
   > n13 = Switch on EAppReturnType(Selection=@n12)
   n14 = getSelectedData(execute=[@n13.Yes, @n11.else], self=Get TL_FilteredList())
   > SpawnLevelsFromListData(Array=@n14)
   n15 = "On Selection Changed (DD_SeasonSelector)"!() -> (OutputDelegate, then, SelectedItem:String, SelectionType:ESelectInfo Enum) {is=title K2Node_ComponentBoundEvent}
   > clearItems(self=Get TL_FilteredList())
   > Set SearchFilter(SearchFilter=@n15.SelectedItem)
   > GetFilteredSublevels()
   > PopulateListfromArray(Array=Get FilteredLevelsList())
   "On Clicked (BTN_WorkflowSwap)"!() -> (OutputDelegate, then, widget:User Widget Object Reference, event:Pointer Event Structure) {is=title K2Node_ComponentBoundEvent}
   n16 = Get WS_WorkflowSwap()
   > n17 = Branch(Condition="Equal (Integer)"(A=GetActiveWidgetIndex(self=@n16) -> (ReturnValue=0)))
   > clearItems(self=Get TL_FilteredList())
   n18 = Get WS_WorkflowSwap()
   > SetActiveWidgetIndex(self=@n18, Index=1)
   > GetAllCollections()
   > PopulateListfromCollections(Array=Get AllCollectionList() -> (AllCollectionList:Array of Names))
   n19 = Get BTN_WorkflowSwap() -> (BTN_WorkflowSwap:TA Generic 01 Button Object Reference)
   > setLabel!(execute, self:TA Base Editor Button Widget Object Reference=@n19, label:String=Show Levels) -> (then)
   clearItems(execute=@n17.else, self=Get TL_FilteredList())
   > SetActiveWidgetIndex(self=@n18, Index=0)
   > GetAllSublevels()
   > PopulateListfromArray(Array=Get AllLevelsList() -> (AllLevelsList:Array of Names))
   > setLabel(self=@n19, label=Show Collections)
   "On Clicked (BTN_SpawnSelectedCollections)"!() -> (OutputDelegate, then, widget:User Widget Object Reference, event:Pointer Event Structure) {is=title K2Node_ComponentBoundEvent}
   > n20 = getSelectedData(self=Get TL_FilteredList())
   > n21 = Branch(Condition="integer >= integer"(A=Length(TargetArray=@n20), B=20))
   > n22 = Show Message Dialog(Title=Get ModalPopupTitle(), Message=Get ModalPopupMessageText(), MessageType=YesNo, DefaultValue=No, MessageCategory=Warning)
   > n23 = Switch on EAppReturnType(Selection=@n22)
   n24 = getSelectedData(execute=[@n23.Yes, @n21.else], self=Get TL_FilteredList())
   > SpawnCollectionsFromListData!(execute, self, Array:Array of Object References=@n24) -> (then)
   "On Clicked (BTN_EditWidget)"!() -> (OutputDelegate, then, widget:User Widget Object Reference, event:Pointer Event Structure) {is=title K2Node_ComponentBoundEvent}
   > n25 = GetBlueprintAsset(Object=&self)
   > Execute Python Script!(
    execute
    PythonScript="import unreal\r\n\r\nmy_widget = widget_object\r\nasset_editor_subsystem = unreal.get_editor_subsystem(unreal.AssetEditorSubsystem)\r\nasset_editor_subsystem.open_editor_for_assets([my_widget])"
    widget_object:Blueprint Object Reference=@n25
   ) -> (then, ReturnValue) {inputs [widget_object]}
   Event Construct()
   > n26 = Sequence()
   > n27 = Branch(Condition="Equal Exactly (String)"(A=GetPlatformUserName!() -> (ReturnValue:String), B=brianfasten))
   n28 = Get BTN_EditWidget() -> (BTN_EditWidget:TA Generic 01 Button Object Reference)
   > SetVisibility(self=@n28, InVisibility=Visible)
   SetVisibility(execute=@n27.else, self=@n28, InVisibility=Hidden)
   clearItems(execute=@n26.then_1, self=Get TL_FilteredList())
   > GetAllSublevels()
   > PopulateListfromArray(Array=Get AllLevelsList())
   For Each Loop(Array:Array of Wildcards) -> (Array Element:Wildcard)
   > n29 = GetAssetsInCollection(self=Get AssetTagsSubsystem(), Name="Get (a copy)"(Array:Array of Names=Get FilteredCollectionList()) -> (Output:Name))
   > n30 = For Each Loop(Array:Array of Asset Data Structures=@n29) -> (Array Element:Asset Data Structure)
   n31 = int * int(A=@n30.Array Index, B=1024)
   > SpawnActorFromObject(ObjectToUse=GetAsset(InAssetData=@n30), Location=MakeVector(X="To Float (Integer)"(InInt=@n31), Y=0), bTransient=false)
   GetDisplayName()
   DeleteMe_SpawnCollectionStuff() {is=title K2Node_CallFunction}
   GetDisplayName()
   GetDisplayName()
   GetDisplayName()
   GetDisplayName()
   "On Clicked (BTN_SpawnAllCollections)"!() -> (OutputDelegate, then, widget:User Widget Object Reference, event:Pointer Event Structure) {is=title K2Node_ComponentBoundEvent}
  ]}
 GetAllCollections {nodes [
   GetAllCollections() {is=title K2Node_FunctionEntry}
   > n1 = GetCollections!(execute, self:Asset Tags Subsystem Object Reference=Get AssetTagsSubsystem()) -> (then, ReturnValue:Array of Names)
   > n2 = Set AllCollectionList(AllCollectionList:Array of Names=@n1) -> (Output_Get:Array of Names)
   > n3 = For Each Loop(Array:Array of Names=@n2) -> (Array Element:Name)
   n4 = "To String (Name)"(InName=@n3)
   > Branch(Condition=StartsWith(SourceString=@n4))
   > addItemGeneric!(execute, self:TA Generic 01 List Object Reference=Get TL_FilteredList(), displayString:String=@n4) -> (then, widget:User Widget Object Reference, data:Object Reference)
  ]}
 GetAllSublevels {nodes [
   GetAllSublevels() {is=title K2Node_FunctionEntry}
   > Clear(TargetArray:Array of Names=Get AllLevelsList())
   n1 = MakeTopLevelAssetPath!(PackageName:String=/Script/Engine, AssetName:String=World) -> (ReturnValue:Top Level Asset Path Structure)
   > n2 = GetAssets!(
    execute
    self:Asset Registry Interface=GetAssetRegistry()
    Filter:"ARFilter Structure (by ref)"=MakeARFilter!(
     PackageNames:Array of Names
     PackagePaths:Array of Names=Make Array("[0]":Name=/Game/Environments/Outsource/Sublevels/) -> (Array:Array of Names)
     SoftObjectPaths:Array of Soft Object Path Structures
     ClassPaths:Array of Top Level Asset Path Structures=Make Array("[0]":Top Level Asset Path Structure=@n1) -> (Array:Array of Top Level Asset Path Structures)
     RecursiveClassPathsExclusionSet:Set of Top Level Asset Path Structures
     ClassNames:Array of Names
     RecursiveClassesExclusionSet:Set of Names
     bRecursivePaths:bool=true
     bRecursiveClasses:bool=false
     bIncludeOnlyOnDiskAssets:bool=false
    ) -> (ReturnValue:ARFilter Structure)
    bSkipARFilteredAssets:bool=true
   ) -> (then, OutAssetData:Array of Asset Data Structures, ReturnValue:bool)
   > n3 = For Each Loop(Array:Array of Asset Data Structures=@n2) -> (Array Element:Asset Data Structure)
   > Add(TargetArray:Array of Names=Get AllLevelsList(), NewItem:"Name (by ref)"=Break Asset Data(AssetData=@n3)) {is=title K2Node_CallArrayFunction}
  ]}
 GetFilteredCollections {nodes [
   GetFilteredCollections() {is=title K2Node_FunctionEntry}
   n1 = Get SearchFilter() -> (SearchFilter:String)
   > n2 = Branch(Condition="integer > integer"(A=Len(S=@n1)))
   > Clear(TargetArray=Get FilteredCollectionList())
   > n3 = GetCollections(self=Get AssetTagsSubsystem())
   > n4 = For Each Loop(Array:Array of Names=@n3) -> (Array Element:Name)
   > Branch(Condition=Contains(SearchIn="To String (Name)"(InName=@n4), Substring=Get SearchFilter()))
   > Add(TargetArray=Get FilteredCollectionList(), NewItem=@n4)
   n5 = Get FilteredCollectionList()
   Clear(execute=@n2.else, TargetArray=@n5)
   > n6 = For Each Loop(Array=Get AllCollectionList())
   > Add(TargetArray=@n5, NewItem=@n6)
  ]}
 GetFilteredSublevels {nodes [
   GetFilteredSublevels() {is=title K2Node_FunctionEntry}
   n1 = Get SearchFilter()
   > n2 = Branch(Condition="integer > integer"(A=Len(S=@n1)))
   > Clear(TargetArray=Get FilteredLevelsList())
   n3 = MakeTopLevelAssetPath(PackageName=&Engine, AssetName=World)
   > n4 = GetAssets(
    self=GetAssetRegistry()
    Filter=MakeARFilter(
     PackagePaths=Make Array("[0]":Name=&Sublevels/) -> (Array:Array of Names)
     ClassPaths=Make Array("[0]":Top Level Asset Path Structure=@n3) -> (Array:Array of Top Level Asset Path Structures)
     bRecursivePaths=true
     bRecursiveClasses=false
     bIncludeOnlyOnDiskAssets=false
    )
    bSkipARFilteredAssets=true
   )
   > n5 = For Each Loop(Array:Array of Asset Data Structures=@n4) -> (Array Element:Asset Data Structure)
   n6 = Break Asset Data(AssetData=@n5)
   > Branch(Condition=Contains(SearchIn="To String (Name)"(InName=@n6.AssetName), Substring=Get SearchFilter()))
   > Add(TargetArray=Get FilteredLevelsList(), NewItem=@n6)
   n7 = Get FilteredLevelsList()
   Clear(execute=@n2.else, TargetArray=@n7)
   > n8 = For Each Loop(Array:Array of Names=Get AllLevelsList()) -> (Array Element:Name)
   > Add(TargetArray=@n7, NewItem=@n8)
  ]}
 PopulateListfromArray {nodes [
   n1 = PopulateListfromArray!() -> (then, Array:Array of Names) {is=title K2Node_FunctionEntry}
   > n2 = For Each Loop(Array=@n1)
   > n3 = Construct Object from Class(Class=/Game/Blueprints/Blutilities/ForTechArt/Data/TA_GenericListItemData.TA_GenericListItemData_C) -> (ReturnValue:TA Generic List Item Data Object Reference)
   n4 = "To String (Name)"(InName=@n2)
   > Set strings!(
    execute
    strings:Array of Strings=Make Array("[0]":String=@n4) -> (Array:Array of Strings)
    self:TA Generic List Item Data Object Reference=@n3
   ) -> (then, Output_Get:Array of Strings)
   > Set displayString!(
    execute
    displayString:String=Split(SourceString=@n4, InStr=/, SearchDir=FromEnd).RightS
    self:TA Generic List Item Data Object Reference=@n3
   ) -> (then, Output_Get:String)
   > addItem!(execute, self:TA Generic 01 List Object Reference=Get TL_FilteredList(), dataObject:Object Reference=@n3) -> (then, newItem:User Widget Object Reference, newData:Object Reference)
  ]}
 PopulateListfromCollections {nodes [
   n1 = PopulateListfromCollections!() -> (then, Array:Array of Names) {is=title K2Node_FunctionEntry}
   > n2 = For Each Loop(Array=@n1)
   > n3 = Construct Object from Class(Class=&TA_GenericListItemData_C)
   > Set displayString(displayString="To String (Name)"(InName=@n2), self=@n3)
   > addItem(self=Get TL_FilteredList(), dataObject=@n3)
  ]}
 PreviewAllFoundations {nodes [
   PreviewAllFoundations() {is=title K2Node_FunctionEntry}
   > n1 = GetAllActorsOfClass!(
    execute
    WorldContextObject:Object Reference
    ActorClass:Actor Class Reference=/Game/Athena/Environments/Blueprints/LF_1x1_Foundation.LF_1x1_Foundation_C
   ) -> (then, OutActors:Array of LF 1x 1 Foundation Object References)
   > n2 = For Each Loop(Array:Array of LF 1x 1 Foundation Object References=@n1) -> (Array Element:LF 1x 1 Foundation Object Reference)
   > SetEditorProperty(Object=@n2, PropertyName=bPreviewFoundationLevelInEditor, PropertyValue=MakeLiteralBool(Value=true))
   Cast To BuildingFoundation!(execute, Object:Wildcard) -> (then, CastFailed, AsBuilding Foundation:Building Foundation Object Reference) {is=title K2Node_DynamicCast}
  ]}
 RemoveAllAssets {nodes [
   RemoveAllAssets() {is=title K2Node_FunctionEntry}
   > n1 = Sequence!(execute) -> (then_0, then_1, then_2:Exec, then_3:Exec)
   > n2 = GetAllActorsOfClass(ActorClass=&LF_1x1_Foundation_C)
   > Destroy Actor(self=@n2)
   n3 = GetAllActorsOfClass(execute=@n1.then_1, ActorClass=/Script/Engine.TextRenderActor) -> (OutActors:Array of Text Render Actor Object References)
   > Destroy Actor(self=@n3)
   n4 = GetAllActorsOfClass(execute=@n1.then_2, ActorClass=/Script/FortniteGame.BuildingActor) -> (OutActors:Array of Building Actor Object References)
   > Destroy Actor(self=@n4)
   Set SpawnCount(execute=@n1.then_3, SpawnCount=0) -> (Output_Get=0)
   > PrintString(InString=Map Reset, TextColor=(R=1,G=0,B=0,A=1), Duration=5)
  ]}
 Spawn Text Actor {nodes [
   n1 = Spawn Text Actor!() -> (then, Location:Vector=~, SourceString:String, Rotation:Rotator=~) {is=title K2Node_FunctionEntry}
   > n2 = SpawnActorFromClass!(
    execute
    ActorClass:Actor Class Reference=&TextRenderActor
    Location:Vector=@n1
    Rotation:Rotator=@n1.Rotation
    bTransient:bool
   ) -> (then, ReturnValue:Text Render Actor Object Reference)
   n3 = Split(SourceString=@n1.SourceString, InStr=/, SearchDir=FromEnd)
   n4 = Get TextRender!(self:Text Render Actor Object Reference=@n2) -> (TextRender:Text Render Component Object Reference)
   > Set Text!(
    execute
    self:Text Render Component Object Reference=@n4
    Value:"Text (by ref)"=TextToUpper!(InText:"Text (by ref)"="To Text (String)"(InString=@n3.RightS)) -> (ReturnValue:Text)
   ) -> (then) {is=title K2Node_CallFunction}
   > SetWorldSize!(execute, self:Text Render Component Object Reference=@n4, Value:float=Get L_TextSize() -> (L_TextSize:float)) -> (then)
   > SetTextMaterial!(
    execute
    self:Text Render Component Object Reference=@n4
    Material:Material Interface Object Reference=/Engine/EngineMaterials/DefaultTextMaterialTranslucent
   ) -> (then)
   > SetTextRenderColor!(execute, self:Text Render Component Object Reference=@n4, Value:Color Structure=Get L_TextColor() -> (L_TextColor:Color Structure)) -> (then)
   > Set Actor Hidden In Game!(execute, self:Actor Object Reference=@n2, bNewHidden:bool=true) -> (then) {is=title K2Node_CallFunction}
  ]}
 SpawnAssetsFromCollection {nodes [
   n1 = SpawnAssetsFromCollection!() -> (then, Array:Array of Strings) {is=title K2Node_FunctionEntry}
   > n2 = For Each Loop(Array:Array of Strings=@n1) -> (Array Element:String)
   > n3 = GetAssetsInCollection(self=Get AssetTagsSubsystem(), Name=String To Name(InString=@n2))
   > n4 = For Each Loop(Array:Array of Asset Data Structures=@n3) -> (Array Element:Asset Data Structure)
   n5 = float * float(A=Get SpawnCount() -> (SpawnCount=0), B:int=1024)
   n6 = BreakVector(InVec=Get StartLocation())
   n7 = float * float(A=@n4.Array Index, B:double=1024)
   > SpawnActorFromObject(
    ObjectToUse=GetAsset(InAssetData=@n4)
    Location=MakeVector(X=float + float(A=@n6, B=@n5), Y=float - float(A=@n6.Y, B=@n7))
    Rotation=Get Collection_StartRotation()
    bTransient=false
   )
   n8 = BreakVector(InVec=Get StartLocation())
   n9 = float + float(A=@n8, B=float * float(A=Get SpawnCount() -> (SpawnCount=0), B=1024))
   n10 = BreakRotator(InRot=Get StartRotation())
   SpawnTextActor_FromCollection(
    execute=@n4.Completed
    Location=MakeVector(X=float - float(A=@n9, B=512), Y=@n8.Y, Z=1)
    SourceString=@n2
    Rotation=MakeRotator(Roll=@n10, Pitch=90, Yaw=@n10.Yaw)
   )
   > Set SpawnCount(SpawnCount=int + int(A=Get SpawnCount() -> (SpawnCount=0))) -> (Output_Get=0)
  ]}
 SpawnCollectionsFromListData {nodes [
   n1 = SpawnCollectionsFromListData!() -> (then, Array:Array of Object References) {is=title K2Node_FunctionEntry}
   > n2 = For Each Loop(Array:Array of Object References=@n1) -> (Array Element:Object Reference)
   > n3 = Cast To TA_GenericListItemData!(execute, Object=@n2) -> (then, CastFailed, AsTA Generic List Item Data:TA Generic List Item Data Object Reference) {is=title K2Node_DynamicCast}
   > SpawnAssetsFromCollection(Array=Make Array("[0]"=Get displayString!(self:TA Generic List Item Data Object Reference=@n3) -> (displayString:String))) {is=title K2Node_CallFunction}
  ]}
 SpawnLevelsFromListData {nodes [
   n1 = SpawnLevelsFromListData!() -> (then, Array:Array of Object References) {is=title K2Node_FunctionEntry}
   > n2 = For Each Loop(Array=@n1)
   PreviewAllFoundations(execute=@n2.Completed) {is=title K2Node_CallFunction}
   > PrintString(InString=All Levels Loaded, Duration=5)
   n3 = Cast To TA_GenericListItemData(execute=@n2.LoopBody, Object=@n2)
   n4 = Get strings!(self:TA Generic List Item Data Object Reference=@n3) -> (strings:Array of Strings)
   > SpawnSublevel!(execute, self, LevelList:Name=String To Name(InString="Get (a copy)"(Array:Array of Strings=@n4) -> (Output:String)), IterationCount:int=0) -> (then)
  ]}
 SpawnSelectedLevels {nodes [
   SpawnSelectedLevels() {is=title K2Node_FunctionEntry}
   > For Each Loop(Array:Array of Names=Get FilteredLevelsList()) -> (Array Element:Name)
   > SpawnSublevel(LevelList=None, IterationCount=0)
  ]}
 SpawnSublevel {nodes [
   n1 = SpawnSublevel!() -> (then, LevelList:Name, IterationCount:int) {is=title K2Node_FunctionEntry}
   > n2 = Sequence!(execute) -> (then_0, then_1)
   n3 = "To String (Name)"(InName=@n1)
   n4 = BreakVector(InVec=Get StartLocation())
   n5 = Get SpawnCount() -> (SpawnCount=0)
   n6 = MakeVector(X=float + float(A=@n4, B=float * float(A=@n5, B:int=1024)), Y=@n4.Y, Z=@n4.Z)
   n7 = BreakVector(InVec=@n6)
   n8 = BreakRotator(InRot=Get StartRotation())
   Spawn Text Actor!(
    execute=@n2.then_1
    self
    Location:Vector=MakeVector(X=float - float(A=@n7, B=512), Y=@n7.Y, Z=1)
    SourceString:String=@n3
    Rotation:Rotator=MakeRotator(Roll=@n8, Pitch=90, Yaw=@n8.Yaw)
   ) -> (then) {is=title K2Node_CallFunction}
   > Set SpawnCount(SpawnCount=int + int(A=Get SpawnCount() -> (SpawnCount=0))) -> (Output_Get=0)
   n9 = SpawnActorFromObject(
    execute=@n2.then_0
    ObjectToUse=/Game/Athena/Environments/Blueprints/LF_1x1_Foundation
    Location=@n6
    Rotation=Get StartRotation()
    bTransient=false
   )
   > n10 = Cast To BuildingFoundation(Object:Object Reference=@n9)
   > EditorOnlyAddAdditionalWorld!(execute, self:Building Foundation Object Reference=@n10, LevelPackageName:"String (by ref)"=@n3) -> (then)
  ]}
 SpawnTextActor_FromCollection {nodes [
   n1 = SpawnTextActor_FromCollection!() -> (then, Location:Vector=~, SourceString:String, Rotation:Rotator=~) {is=title K2Node_FunctionEntry}
   > n2 = SpawnActorFromClass(ActorClass=&TextRenderActor, Location=@n1, Rotation=@n1.Rotation)
   n3 = Get TextRender(self=@n2)
   > Set Text(self=@n3, Value=TextToUpper(InText="To Text (String)"(InString=@n1.SourceString)))
   > SetWorldSize(self=@n3, Value=Get L_TextSize())
   > SetTextMaterial(self=@n3, Material=&DefaultTextMaterialTranslucent)
   > SetTextRenderColor(self=@n3, Value=Get L_TextColor())
   > Set Actor Hidden In Game(self=@n2, bNewHidden=true)
  ]}
]
variables [
 AllCollectionList {type="array<string>", default []}
 AllLevelsList {type="array<string>", default []}
 Collection_StartRotation {type=Rotator, default=<rot 0 90 0>}
 DeleteMe_CollectionName {type=string, default=ENV_Graybox_BPs}
 FilteredCollectionList {type="array<string>", default []}
 FilteredLevelsList {type="array<string>", default []}
 InvalidFilterResults {type=string, default=No Levels Found}
 ModalPopupMessageText {
  type=string
  default="You are about to load a large number of levels. This will make the editor unresponsive until operation is completed. \r\n\r\nContinue?"
 }
 ModalPopupResult {type=EAppReturnType, default=No}
 ModalPopupTitle {type=string, default="Warning!"}
 SearchFilter {type=string, default=""}
 SelectedAssets {type="array<BuildingActor>", default []}
 SpawnCount {type=integer}
 StartLocation {type=Vector, default=<-12288 12288 0>}
 StartRotation {type=Rotator}
]
widgets {
 info {rootWidgetClass=&HorizontalBox, widgetCount=23}
 widgets [
  HorizontalBox_0 {widgetClassPath=&HorizontalBox, widget=HorizontalBox_0}
  VerticalBox_3 {widgetClassPath=&VerticalBox, parent=HorizontalBox_0, widget=VerticalBox_3, slot {class=HorizontalBoxSlot}}
  BTN_WorkflowSwap {
   widgetClassPath=&TA_Generic01_Button
   parent=VerticalBox_3
   widget=BTN_WorkflowSwap
   properties {label=Show Collections, toolTipText=Spawn actors using saved collections or sublevels, buttonColor_Default=<rgba 0.6354 0.4544 0 1>}
   slot {class=VerticalBoxSlot, properties {verticalAlignment=VAlign_Top, padding=<ltrb 5 5 5 5>}}
  }
  CommonBorder_396 {
   widgetClassPath=&CommonBorder
   parent=VerticalBox_3
   widget=CommonBorder_396
   properties {
    style=/FortUILibrary/Styles/Border/Border-SolidBG.Border-SolidBG_C
    verticalAlignment=VAlign_Top
    brushColor=<rgba 0 0 0 10>
    padding=<ltrb 0 0 0 0>
   }
   slot {class=VerticalBoxSlot}
  }
  WS_WorkflowSwap {
   widgetClassPath=&WidgetSwitcher
   bIsVariable=true
   parent=CommonBorder_396
   widget=WS_WorkflowSwap
   slot {class=BorderSlot, properties {verticalAlignment=VAlign_Top, padding=<ltrb 0 0 0 0>}}
  }
  Sublevel_Workflow {widgetClassPath=&VerticalBox, parent=WS_WorkflowSwap, widget=Sublevel_Workflow, slot {class=WidgetSwitcherSlot}}
  VerticalBox_2 {widgetClassPath=&VerticalBox, parent=Sublevel_Workflow, widget=VerticalBox_2, slot {class=VerticalBoxSlot}}
  CommonTextBlock {
   widgetClassPath=&CommonTextBlock
   parent=VerticalBox_2
   widget=CommonTextBlock
   properties {
    style=/FortUILibrary/Styles/Typography/Foundation/TextStyle-Base-S.TextStyle-Base-S_C
    text=Season Select
    font {
     bForceMonospaced=false
     bMaterialIsStencil=false
     fontMaterial=null
     fontObject=&NotoSans
     letterSpacing=0
     monospacedWidth=1
     size=18
     skewAmount=0
     typefaceFontName=NotoSans-Regular
     outlineSettings {
      bApplyOutlineToDropShadows=false
      bMiteredCorners=false
      bSeparateFillAlpha=false
      outlineMaterial=null
      outlineSize=0
      outlineColor=<rgba 0 0 0 1>
     }
    }
    shadowOffset=<0 0>
   }
   slot {class=VerticalBoxSlot, properties {horizontalAlignment=HAlign_Center, padding=<ltrb 4 2 4 2>}}
  }
  DD_SeasonSelector {
   widgetClassPath=/Script/UMG.ComboBoxString
   bIsVariable=true
   parent=VerticalBox_2
   widget=DD_SeasonSelector
   properties {
    selectedOption=Season Select
    defaultOptions [10,11,12,13,14,15,16,17,18,19,20,21,22,23,24]
    foregroundColor {colorUseRule=UseColor_Specified, specifiedColor=<rgba 0.5271 0.5271 0.5271 1>}
    itemStyle {
     bUseParentRowBrush=false
     activeBrush {
      bIsDynamicallyLoaded=false
      drawAs=Image
      imageType=NoImage
      mirroring=NoMirror
      resourceName=None
      resourceObject=null
      tiling=NoTile
      imageSize=<0 0>
      margin=<ltrb 0 0 0 0>
      outlineSettings {
       bUseBrushTransparency=false
       roundingType=HalfHeightRadius
       width=0
       color {colorUseRule=UseColor_Specified, specifiedColor=<rgba 0 0 0 0>}
       cornerRadii=<0 0 0 0>
      }
      tintColor {colorUseRule=UseColor_Specified, specifiedColor=<rgba 0 0.162 0.7454 1>}
      uVRegion {bIsValid=false, max=<0 0>, min=<0 0>}
     }
     activeHighlightedBrush {
      bIsDynamicallyLoaded=false
      drawAs=Image
      imageType=NoImage
      mirroring=NoMirror
      resourceName=None
      resourceObject=null
      tiling=NoTile
      imageSize=<0 0>
      margin=<ltrb 0 0 0 0>
      outlineSettings {
       bUseBrushTransparency=false
       roundingType=HalfHeightRadius
       width=0
       color {colorUseRule=UseColor_Specified, specifiedColor=<rgba 0 0 0 0>}
       cornerRadii=<0 0 0 0>
      }
      tintColor {colorUseRule=UseColor_Specified, specifiedColor=<rgba 0.0252 0.0319 0.0423 1>}
      uVRegion {bIsValid=false, max=<0 0>, min=<0 0>}
     }
     activeHoveredBrush {
      bIsDynamicallyLoaded=false
      drawAs=Image
      imageType=NoImage
      mirroring=NoMirror
      resourceName=None
      resourceObject=null
      tiling=NoTile
      imageSize=<0 0>
      margin=<ltrb 0 0 0 0>
      outlineSettings {
       bUseBrushTransparency=false
       roundingType=HalfHeightRadius
       width=0
       color {colorUseRule=UseColor_Specified, specifiedColor=<rgba 0 0 0 0>}
       cornerRadii=<0 0 0 0>
      }
      tintColor {colorUseRule=UseColor_Specified, specifiedColor=<rgba 0 0.162 0.7454 1>}
      uVRegion {bIsValid=false, max=<0 0>, min=<0 0>}
     }
     dropIndicator_Above {
      bIsDynamicallyLoaded=false
      drawAs=Box
      imageType=FullColor
      mirroring=NoMirror
      resourceName=../../../Engine/Content/Slate/Common/DropZoneIndicator_Above.png
      resourceObject=null
      tiling=NoTile
      imageSize=<0 0>
      margin=<ltrb 0.625 0.625 0 0>
      outlineSettings {
       bUseBrushTransparency=false
       roundingType=HalfHeightRadius
       width=0
       color {colorUseRule=UseColor_Specified, specifiedColor=<rgba 0 0 0 0>}
       cornerRadii=<0 0 0 0>
      }
      tintColor {colorUseRule=UseColor_Specified, specifiedColor=<rgba 0.0091 0.3515 0.9647 1>}
      uVRegion {bIsValid=false, max=<0 0>, min=<0 0>}
     }
     dropIndicator_Below {
      bIsDynamicallyLoaded=false
      drawAs=Box
      imageType=FullColor
      mirroring=NoMirror
      resourceName=../../../Engine/Content/Slate/Common/DropZoneIndicator_Below.png
      resourceObject=null
      tiling=NoTile
      imageSize=<0 0>
      margin=<ltrb 0.625 0 0 0.625>
      outlineSettings {
       bUseBrushTransparency=false
       roundingType=HalfHeightRadius
       width=0
       color {colorUseRule=UseColor_Specified, specifiedColor=<rgba 0 0 0 0>}
       cornerRadii=<0 0 0 0>
      }
      tintColor {colorUseRule=UseColor_Specified, specifiedColor=<rgba 0.0091 0.3515 0.9647 1>}
      uVRegion {bIsValid=false, max=<0 0>, min=<0 0>}
     }
     dropIndicator_Onto {
      bIsDynamicallyLoaded=false
      drawAs=Box
      imageType=FullColor
      mirroring=NoMirror
      resourceName=../../../Engine/Content/Slate/Common/DropZoneIndicator_Onto.png
      resourceObject=null
      tiling=NoTile
      imageSize=<0 0>
      margin=<ltrb 0.25 0.25 0.25 0.25>
      outlineSettings {
       bUseBrushTransparency=false
       roundingType=HalfHeightRadius
       width=0
       color {colorUseRule=UseColor_Specified, specifiedColor=<rgba 0 0 0 0>}
       cornerRadii=<0 0 0 0>
      }
      tintColor {colorUseRule=UseColor_Specified, specifiedColor=<rgba 0.0091 0.3515 0.9647 1>}
      uVRegion {bIsValid=false, max=<0 0>, min=<0 0>}
     }
     evenRowBackgroundBrush {
      bIsDynamicallyLoaded=false
      drawAs=Image
      imageType=NoImage
      mirroring=NoMirror
      resourceName=None
      resourceObject=null
      tiling=NoTile
      imageSize=<0 0>
      margin=<ltrb 0 0 0 0>
      outlineSettings {
       bUseBrushTransparency=false
       roundingType=HalfHeightRadius
       width=0
       color {colorUseRule=UseColor_Specified, specifiedColor=<rgba 0 0 0 0>}
       cornerRadii=<0 0 0 0>
      }
      tintColor {colorUseRule=UseColor_Specified, specifiedColor=<rgba 0.0103 0.0103 0.0103 1>}
      uVRegion {bIsValid=false, max=<0 0>, min=<0 0>}
     }
     evenRowBackgroundHoveredBrush {
      bIsDynamicallyLoaded=false
      drawAs=Image
      imageType=NoImage
      mirroring=NoMirror
      resourceName=None
      resourceObject=null
      tiling=NoTile
      imageSize=<0 0>
      margin=<ltrb 0 0 0 0>
      outlineSettings {
       bUseBrushTransparency=false
       roundingType=HalfHeightRadius
       width=0
       color {colorUseRule=UseColor_Specified, specifiedColor=<rgba 0 0 0 0>}
       cornerRadii=<0 0 0 0>
      }
      tintColor {colorUseRule=UseColor_Specified, specifiedColor=<rgba 0.0176 0.0176 0.0176 1>}
      uVRegion {bIsValid=false, max=<0 0>, min=<0 0>}
     }
     inactiveBrush {
      bIsDynamicallyLoaded=false
      drawAs=Image
      imageType=NoImage
      mirroring=NoMirror
      resourceName=None
      resourceObject=null
      tiling=NoTile
      imageSize=<0 0>
      margin=<ltrb 0 0 0 0>
      outlineSettings {
       bUseBrushTransparency=false
       roundingType=HalfHeightRadius
       width=0
       color {colorUseRule=UseColor_Specified, specifiedColor=<rgba 0 0 0 0>}
       cornerRadii=<0 0 0 0>
      }
      tintColor {colorUseRule=UseColor_Specified, specifiedColor=<rgba 0.0513 0.0953 0.159 1>}
      uVRegion {bIsValid=false, max=<0 0>, min=<0 0>}
     }
     inactiveHighlightedBrush {
      bIsDynamicallyLoaded=false
      drawAs=Image
      imageType=NoImage
      mirroring=NoMirror
      resourceName=None
      resourceObject=null
      tiling=NoTile
      imageSize=<0 0>
      margin=<ltrb 0 0 0 0>
      outlineSettings {
       bUseBrushTransparency=false
       roundingType=HalfHeightRadius
       width=0
       color {colorUseRule=UseColor_Specified, specifiedColor=<rgba 0 0 0 0>}
       cornerRadii=<0 0 0 0>
      }
      tintColor {colorUseRule=UseColor_Specified, specifiedColor=<rgba 0.0252 0.0319 0.0423 1>}
      uVRegion {bIsValid=false, max=<0 0>, min=<0 0>}
     }
     inactiveHoveredBrush {
      bIsDynamicallyLoaded=false
      drawAs=Image
      imageType=NoImage
      mirroring=NoMirror
      resourceName=None
      resourceObject=null
      tiling=NoTile
      imageSize=<0 0>
      margin=<ltrb 0 0 0 0>
      outlineSettings {
       bUseBrushTransparency=false
       roundingType=HalfHeightRadius
       width=0
       color {colorUseRule=UseColor_Specified, specifiedColor=<rgba 0 0 0 0>}
       cornerRadii=<0 0 0 0>
      }
      tintColor {colorUseRule=UseColor_Specified, specifiedColor=<rgba 0.0513 0.0953 0.159 1>}
      uVRegion {bIsValid=false, max=<0 0>, min=<0 0>}
     }
     oddRowBackgroundBrush {
      bIsDynamicallyLoaded=false
      drawAs=Image
      imageType=NoImage
      mirroring=NoMirror
      resourceName=None
      resourceObject=null
      tiling=NoTile
      imageSize=<0 0>
      margin=<ltrb 0 0 0 0>
      outlineSettings {
       bUseBrushTransparency=false
       roundingType=HalfHeightRadius
       width=0
       color {colorUseRule=UseColor_Specified, specifiedColor=<rgba 0 0 0 0>}
       cornerRadii=<0 0 0 0>
      }
      tintColor {colorUseRule=UseColor_Specified, specifiedColor=<rgba 0.0103 0.0103 0.0103 1>}
      uVRegion {bIsValid=false, max=<0 0>, min=<0 0>}
     }
     oddRowBackgroundHoveredBrush {
      bIsDynamicallyLoaded=false
      drawAs=Image
      imageType=NoImage
      mirroring=NoMirror
      resourceName=None
      resourceObject=null
      tiling=NoTile
      imageSize=<0 0>
      margin=<ltrb 0 0 0 0>
      outlineSettings {
       bUseBrushTransparency=false
       roundingType=HalfHeightRadius
       width=0
       color {colorUseRule=UseColor_Specified, specifiedColor=<rgba 0 0 0 0>}
       cornerRadii=<0 0 0 0>
      }
      tintColor {colorUseRule=UseColor_Specified, specifiedColor=<rgba 0.0176 0.0176 0.0176 1>}
      uVRegion {bIsValid=false, max=<0 0>, min=<0 0>}
     }
     parentRowBackgroundBrush {
      bIsDynamicallyLoaded=false
      drawAs=Image
      imageType=NoImage
      mirroring=NoMirror
      resourceName=None
      resourceObject=null
      tiling=NoTile
      imageSize=<32 32>
      margin=<ltrb 0 0 0 0>
      outlineSettings {
       bUseBrushTransparency=false
       roundingType=HalfHeightRadius
       width=0
       color {colorUseRule=UseColor_Specified, specifiedColor=<rgba 0 0 0 0>}
       cornerRadii=<0 0 0 0>
      }
      tintColor {colorUseRule=UseColor_Specified, specifiedColor=<rgba 1 1 1 1>}
      uVRegion {bIsValid=false, max=<0 0>, min=<0 0>}
     }
     parentRowBackgroundHoveredBrush {
      bIsDynamicallyLoaded=false
      drawAs=Image
      imageType=NoImage
      mirroring=NoMirror
      resourceName=None
      resourceObject=null
      tiling=NoTile
      imageSize=<32 32>
      margin=<ltrb 0 0 0 0>
      outlineSettings {
       bUseBrushTransparency=false
       roundingType=HalfHeightRadius
       width=0
       color {colorUseRule=UseColor_Specified, specifiedColor=<rgba 0 0 0 0>}
       cornerRadii=<0 0 0 0>
      }
      tintColor {colorUseRule=UseColor_Specified, specifiedColor=<rgba 1 1 1 1>}
      uVRegion {bIsValid=false, max=<0 0>, min=<0 0>}
     }
     selectedTextColor {colorUseRule=UseColor_Specified, specifiedColor=<rgba 1 1 1 1>}
     selectorFocusedBrush {
      bIsDynamicallyLoaded=false
      drawAs=Border
      imageType=FullColor
      mirroring=NoMirror
      resourceName=../../../Engine/Content/Slate/Common/Selector.png
      resourceObject=null
      tiling=Both
      imageSize=<0 0>
      margin=<ltrb 0.25 0.25 0.25 0.25>
      outlineSettings {
       bUseBrushTransparency=false
       roundingType=HalfHeightRadius
       width=0
       color {colorUseRule=UseColor_Specified, specifiedColor=<rgba 0 0 0 0>}
       cornerRadii=<0 0 0 0>
      }
      tintColor {colorUseRule=UseColor_Specified, specifiedColor=<rgba 0.701 0.225 0.003 1>}
      uVRegion {bIsValid=false, max=<0 0>, min=<0 0>}
     }
     textColor {colorUseRule=UseColor_Specified, specifiedColor=<rgba 0.5271 0.5271 0.5271 1>}
    }
    scrollBarStyle {
     thickness=8
     draggedThumbImage {
      bIsDynamicallyLoaded=false
      drawAs=RoundedBox
      imageType=NoImage
      mirroring=NoMirror
      resourceName=None
      resourceObject=null
      tiling=NoTile
      imageSize=<0 0>
      margin=<ltrb 0 0 0 0>
      outlineSettings {
       bUseBrushTransparency=false
       roundingType=FixedRadius
       width=0
       color {colorUseRule=UseColor_Specified, specifiedColor=<rgba 0 0 0 0>}
       cornerRadii=<4 4 4 4>
      }
      tintColor {colorUseRule=UseColor_Specified, specifiedColor=<rgba 0.2159 0.2159 0.2159 1>}
      uVRegion {bIsValid=false, max=<0 0>, min=<0 0>}
     }
     horizontalBackgroundImage {
      bIsDynamicallyLoaded=false
      drawAs=NoDrawType
      imageType=NoImage
      mirroring=NoMirror
      resourceName=None
      resourceObject=null
      tiling=NoTile
      imageSize=<0 0>
      margin=<ltrb 0 0 0 0>
      outlineSettings {
       bUseBrushTransparency=false
       roundingType=HalfHeightRadius
       width=0
       color {colorUseRule=UseColor_Specified, specifiedColor=<rgba 0 0 0 0>}
       cornerRadii=<0 0 0 0>
      }
      tintColor {colorUseRule=UseColor_Specified, specifiedColor=<rgba 1 1 1 1>}
      uVRegion {bIsValid=false, max=<0 0>, min=<0 0>}
     }
     horizontalBottomSlotImage {
      bIsDynamicallyLoaded=false
      drawAs=NoDrawType
      imageType=NoImage
      mirroring=NoMirror
      resourceName=None
      resourceObject=null
      tiling=NoTile
      imageSize=<0 0>
      margin=<ltrb 0 0 0 0>
      outlineSettings {
       bUseBrushTransparency=false
       roundingType=HalfHeightRadius
       width=0
       color {colorUseRule=UseColor_Specified, specifiedColor=<rgba 0 0 0 0>}
       cornerRadii=<0 0 0 0>
      }
      tintColor {colorUseRule=UseColor_Specified, specifiedColor=<rgba 1 1 1 1>}
      uVRegion {bIsValid=false, max=<0 0>, min=<0 0>}
     }
     horizontalTopSlotImage {
      bIsDynamicallyLoaded=false
      drawAs=NoDrawType
      imageType=NoImage
      mirroring=NoMirror
      resourceName=None
      resourceObject=null
      tiling=NoTile
      imageSize=<0 0>
      margin=<ltrb 0 0 0 0>
      outlineSettings {
       bUseBrushTransparency=false
       roundingType=HalfHeightRadius
       width=0
       color {colorUseRule=UseColor_Specified, specifiedColor=<rgba 0 0 0 0>}
       cornerRadii=<0 0 0 0>
      }
      tintColor {colorUseRule=UseColor_Specified, specifiedColor=<rgba 1 1 1 1>}
      uVRegion {bIsValid=false, max=<0 0>, min=<0 0>}
     }
     hoveredThumbImage {
      bIsDynamicallyLoaded=false
      drawAs=RoundedBox
      imageType=NoImage
      mirroring=NoMirror
      resourceName=None
      resourceObject=null
      tiling=NoTile
      imageSize=<0 0>
      margin=<ltrb 0 0 0 0>
      outlineSettings {
       bUseBrushTransparency=false
       roundingType=FixedRadius
       width=0
       color {colorUseRule=UseColor_Specified, specifiedColor=<rgba 0 0 0 0>}
       cornerRadii=<4 4 4 4>
      }
      tintColor {colorUseRule=UseColor_Specified, specifiedColor=<rgba 0.2159 0.2159 0.2159 1>}
      uVRegion {bIsValid=false, max=<0 0>, min=<0 0>}
     }
     normalThumbImage {
      bIsDynamicallyLoaded=false
      drawAs=RoundedBox
      imageType=NoImage
      mirroring=NoMirror
      resourceName=None
      resourceObject=null
      tiling=NoTile
      imageSize=<0 0>
      margin=<ltrb 0 0 0 0>
      outlineSettings {
       bUseBrushTransparency=false
       roundingType=FixedRadius
       width=0
       color {colorUseRule=UseColor_Specified, specifiedColor=<rgba 0 0 0 0>}
       cornerRadii=<4 4 4 4>
      }
      tintColor {colorUseRule=UseColor_Specified, specifiedColor=<rgba 0.0953 0.0953 0.0953 1>}
      uVRegion {bIsValid=false, max=<0 0>, min=<0 0>}
     }
     verticalBackgroundImage {
      bIsDynamicallyLoaded=false
      drawAs=NoDrawType
      imageType=NoImage
      mirroring=NoMirror
      resourceName=None
      resourceObject=null
      tiling=NoTile
      imageSize=<0 0>
      margin=<ltrb 0 0 0 0>
      outlineSettings {
       bUseBrushTransparency=false
       roundingType=HalfHeightRadius
       width=0
       color {colorUseRule=UseColor_Specified, specifiedColor=<rgba 0 0 0 0>}
       cornerRadii=<0 0 0 0>
      }
      tintColor {colorUseRule=UseColor_Specified, specifiedColor=<rgba 1 1 1 1>}
      uVRegion {bIsValid=false, max=<0 0>, min=<0 0>}
     }
     verticalBottomSlotImage {
      bIsDynamicallyLoaded=false
      drawAs=NoDrawType
      imageType=NoImage
      mirroring=NoMirror
      resourceName=None
      resourceObject=null
      tiling=NoTile
      imageSize=<0 0>
      margin=<ltrb 0 0 0 0>
      outlineSettings {
       bUseBrushTransparency=false
       roundingType=HalfHeightRadius
       width=0
       color {colorUseRule=UseColor_Specified, specifiedColor=<rgba 0 0 0 0>}
       cornerRadii=<0 0 0 0>
      }
      tintColor {colorUseRule=UseColor_Specified, specifiedColor=<rgba 1 1 1 1>}
      uVRegion {bIsValid=false, max=<0 0>, min=<0 0>}
     }
     verticalTopSlotImage {
      bIsDynamicallyLoaded=false
      drawAs=NoDrawType
      imageType=NoImage
      mirroring=NoMirror
      resourceName=None
      resourceObject=null
      tiling=NoTile
      imageSize=<0 0>
      margin=<ltrb 0 0 0 0>
      outlineSettings {
       bUseBrushTransparency=false
       roundingType=HalfHeightRadius
       width=0
       color {colorUseRule=UseColor_Specified, specifiedColor=<rgba 0 0 0 0>}
       cornerRadii=<0 0 0 0>
      }
      tintColor {colorUseRule=UseColor_Specified, specifiedColor=<rgba 1 1 1 1>}
      uVRegion {bIsValid=false, max=<0 0>, min=<0 0>}
     }
    }
    widgetStyle {
     comboButtonStyle {
      downArrowAlign=VAlign_Center
      buttonStyle {
       clickedSlateSound {resourceObject=null}
       disabled {
        bIsDynamicallyLoaded=false
        drawAs=RoundedBox
        imageType=NoImage
        mirroring=NoMirror
        resourceName=None
        resourceObject=null
        tiling=NoTile
        imageSize=<0 0>
        margin=<ltrb 0 0 0 0>
        outlineSettings {
         bUseBrushTransparency=false
         roundingType=FixedRadius
         width=1
         color {colorUseRule=UseColor_ColorTable, specifiedColor=<rgba 0 0 0 0>}
         cornerRadii=<4 4 4 4>
        }
        tintColor {colorUseRule=UseColor_Specified, specifiedColor=<rgba 0.0395 0.0395 0.0395 1>}
        uVRegion {bIsValid=false, max=<0 0>, min=<0 0>}
       }
       disabledForeground {colorUseRule=UseColor_ColorTable, specifiedColor=<rgba 0 0 0 0>}
       hovered {
        bIsDynamicallyLoaded=false
        drawAs=RoundedBox
        imageType=NoImage
        mirroring=NoMirror
        resourceName=None
        resourceObject=null
        tiling=NoTile
        imageSize=<0 0>
        margin=<ltrb 0 0 0 0>
        outlineSettings {
         bUseBrushTransparency=false
         roundingType=FixedRadius
         width=1
         color {colorUseRule=UseColor_ColorTable, specifiedColor=<rgba 0 0 0 0>}
         cornerRadii=<4 4 4 4>
        }
        tintColor {colorUseRule=UseColor_Specified, specifiedColor=<rgba 0.0048 0.0048 0.0048 1>}
        uVRegion {bIsValid=false, max=<0 0>, min=<0 0>}
       }
       hoveredForeground {colorUseRule=UseColor_ColorTable, specifiedColor=<rgba 0 0 0 0>}
       hoveredSlateSound {resourceObject=null}
       normal {
        bIsDynamicallyLoaded=false
        drawAs=RoundedBox
        imageType=NoImage
        mirroring=NoMirror
        resourceName=None
        resourceObject=null
        tiling=NoTile
        imageSize=<0 0>
        margin=<ltrb 0 0 0 0>
        outlineSettings {
         bUseBrushTransparency=false
         roundingType=FixedRadius
         width=1
         color {colorUseRule=UseColor_ColorTable, specifiedColor=<rgba 0 0 0 0>}
         cornerRadii=<4 4 4 4>
        }
        tintColor {colorUseRule=UseColor_Specified, specifiedColor=<rgba 0.0048 0.0048 0.0048 1>}
        uVRegion {bIsValid=false, max=<0 0>, min=<0 0>}
       }
       normalForeground {colorUseRule=UseColor_ColorTable, specifiedColor=<rgba 0 0 0 0>}
       normalPadding=<ltrb 8 1 8 1>
       pressed {
        bIsDynamicallyLoaded=false
        drawAs=RoundedBox
        imageType=NoImage
        mirroring=NoMirror
        resourceName=None
        resourceObject=null
        tiling=NoTile
        imageSize=<0 0>
        margin=<ltrb 0 0 0 0>
        outlineSettings {
         bUseBrushTransparency=false
         roundingType=FixedRadius
         width=1
         color {colorUseRule=UseColor_ColorTable, specifiedColor=<rgba 0 0 0 0>}
         cornerRadii=<4 4 4 4>
        }
        tintColor {colorUseRule=UseColor_Specified, specifiedColor=<rgba 0.0048 0.0048 0.0048 1>}
        uVRegion {bIsValid=false, max=<0 0>, min=<0 0>}
       }
       pressedForeground {colorUseRule=UseColor_ColorTable, specifiedColor=<rgba 0 0 0 0>}
       pressedPadding=<ltrb 8 2 8 0>
       pressedSlateSound {resourceObject=null}
      }
      contentPadding=<ltrb 0 0 0 0>
      downArrowImage {
       bIsDynamicallyLoaded=false
       drawAs=Image
       imageType=Vector
       mirroring=NoMirror
       resourceName=../../../Engine/Content/Slate/Starship/CoreWidgets/ComboBox/wide-chevron-down.svg
       resourceObject=null
       tiling=NoTile
       imageSize=<16 16>
       margin=<ltrb 0 0 0 0>
       outlineSettings {
        bUseBrushTransparency=false
        roundingType=HalfHeightRadius
        width=0
        color {colorUseRule=UseColor_Specified, specifiedColor=<rgba 0 0 0 0>}
        cornerRadii=<0 0 0 0>
       }
       tintColor {colorUseRule=UseColor_Specified, specifiedColor=<rgba 1 1 1 1>}
       uVRegion {bIsValid=false, max=<0 0>, min=<0 0>}
      }
      downArrowPadding=<ltrb 2 2 2 2>
      menuBorderBrush {
       bIsDynamicallyLoaded=false
       drawAs=RoundedBox
       imageType=NoImage
       mirroring=NoMirror
       resourceName=None
       resourceObject=null
       tiling=NoTile
       imageSize=<0 0>
       margin=<ltrb 0 0 0 0>
       outlineSettings {
        bUseBrushTransparency=false
        roundingType=FixedRadius
        width=1
        color {colorUseRule=UseColor_ColorTable, specifiedColor=<rgba 0 0 0 0>}
        cornerRadii=<0 0 0 0>
       }
       tintColor {colorUseRule=UseColor_Specified, specifiedColor=<rgba 0.0395 0.0395 0.0395 1>}
       uVRegion {bIsValid=false, max=<0 0>, min=<0 0>}
      }
      menuBorderPadding=<ltrb 0 0 0 0>
      shadowColorAndOpacity=<rgba 0 0 0 1>
      shadowOffset=<0 0>
     }
     contentPadding=<ltrb 0 0 0 0>
     menuRowPadding=<ltrb 8 3 8 3>
     pressedSlateSound {resourceObject=null}
     selectionChangeSlateSound {resourceObject=null}
    }
   }
   slot {class=VerticalBoxSlot, properties {padding=<ltrb 4 2 4 2>}}
  }
  BTN_SpawnSelectedLevels {
   widgetClassPath=&TA_Generic01_Button
   parent=Sublevel_Workflow
   widget=BTN_SpawnSelectedLevels
   properties {label=Spawn Selected Levels, toolTipText=Loads user-selected levels from list}
   slot {class=VerticalBoxSlot, properties {padding=<ltrb 5 5 5 5>}}
  }
  BTN_SpawnAllLevels {
   widgetClassPath=&TA_Generic01_Button
   parent=Sublevel_Workflow
   widget=BTN_SpawnAllLevels
   properties {label=Spawn All Levels, toolTipText=Spawns all levels currently shown in the list}
   slot {class=VerticalBoxSlot, properties {padding=<ltrb 5 5 5 5>}}
  }
  Collection_Workflow {widgetClassPath=&VerticalBox, parent=WS_WorkflowSwap, widget=Collection_Workflow, slot {class=WidgetSwitcherSlot}}
  VerticalBox {widgetClassPath=&VerticalBox, parent=Collection_Workflow, widget=VerticalBox, slot {class=VerticalBoxSlot}}
  BTN_SpawnSelectedCollections {
   widgetClassPath=&TA_Generic01_Button
   parent=Collection_Workflow
   widget=BTN_SpawnSelectedCollections
   properties {label=Spawn Selected Collections, toolTipText=Loads user-selected levels from list}
   slot {class=VerticalBoxSlot, properties {padding=<ltrb 5 5 5 5>}}
  }
  BTN_SpawnAllCollections {
   widgetClassPath=&TA_Generic01_Button
   parent=Collection_Workflow
   widget=BTN_SpawnAllCollections
   properties {label=Spawn All Collections, toolTipText=Spawns all levels currently shown in the list}
   slot {class=VerticalBoxSlot, properties {padding=<ltrb 5 5 5 5>}}
  }
  BTN_ClearLevels {
   widgetClassPath=&TA_Generic01_Button
   parent=VerticalBox_3
   widget=BTN_ClearLevels
   properties {
    label=Clear All Levels
    toolTipText=Removes all foundations and BuildingActors assets from the map
    buttonColor_Default=<rgba 0.6354 0.4544 0 1>
   }
   slot {class=VerticalBoxSlot, properties {verticalAlignment=VAlign_Bottom, padding=<ltrb 5 5 5 5>, size {sizeRule=Fill, value=1}}}
  }
  CommonBorder_1064 {
   widgetClassPath=&CommonBorder
   parent=HorizontalBox_0
   widget=CommonBorder_1064
   properties {style=/FortUILibrary/Styles/Border/Border_DropShadow.Border_DropShadow_C, padding=<ltrb 0 0 0 0>}
   slot {class=HorizontalBoxSlot, properties {padding=<ltrb 5 5 5 5>, size {sizeRule=Fill, value=1}}}
  }
  VerticalBox_1 {
   widgetClassPath=&VerticalBox
   parent=CommonBorder_1064
   widget=VerticalBox_1
   slot {class=BorderSlot, properties {padding=<ltrb 0 0 0 0>}}
  }
  HorizontalBox_1 {
   widgetClassPath=&HorizontalBox
   parent=VerticalBox_1
   widget=HorizontalBox_1
   slot {class=VerticalBoxSlot, properties {padding=<ltrb 5 5 5 5>}}
  }
  CommonTextBlock_403 {
   widgetClassPath=&CommonTextBlock
   parent=HorizontalBox_1
   widget=CommonTextBlock_403
   properties {
    style=&TextStyle-Base-S_C
    text=Filter:
    font {
     bForceMonospaced=false
     bMaterialIsStencil=false
     fontMaterial=null
     fontObject=&NotoSans
     letterSpacing=0
     monospacedWidth=1
     size=18
     skewAmount=0
     typefaceFontName=NotoSans-Regular
     outlineSettings {
      bApplyOutlineToDropShadows=false
      bMiteredCorners=false
      bSeparateFillAlpha=false
      outlineMaterial=null
      outlineSize=0
      outlineColor=<rgba 0 0 0 1>
     }
    }
    shadowOffset=<0 0>
   }
   slot {class=HorizontalBoxSlot, properties {horizontalAlignment=HAlign_Center, padding=<ltrb 5 5 5 5>}}
  }
  TB_TextFilter {
   widgetClassPath=&EditableTextBox
   bIsVariable=true
   parent=HorizontalBox_1
   widget=TB_TextFilter
   properties {widgetStyle {
     backgroundColor {colorUseRule=UseColor_Specified, specifiedColor=<rgba 1 1 1 1>}
     backgroundImageFocused {
      bIsDynamicallyLoaded=false
      drawAs=RoundedBox
      imageType=NoImage
      mirroring=NoMirror
      resourceName=None
      resourceObject=null
      tiling=NoTile
      imageSize=<0 0>
      margin=<ltrb 0 0 0 0>
      outlineSettings {
       bUseBrushTransparency=false
       roundingType=FixedRadius
       width=1
       color {colorUseRule=UseColor_ColorTable, specifiedColor=<rgba 0 0 0 0>}
       cornerRadii=<4 4 4 4>
      }
      tintColor {colorUseRule=UseColor_Specified, specifiedColor=<rgba 0.0048 0.0048 0.0048 1>}
      uVRegion {bIsValid=false, max=<0 0>, min=<0 0>}
     }
     backgroundImageHovered {
      bIsDynamicallyLoaded=false
      drawAs=RoundedBox
      imageType=NoImage
      mirroring=NoMirror
      resourceName=None
      resourceObject=null
      tiling=NoTile
      imageSize=<0 0>
      margin=<ltrb 0 0 0 0>
      outlineSettings {
       bUseBrushTransparency=false
       roundingType=FixedRadius
       width=1
       color {colorUseRule=UseColor_ColorTable, specifiedColor=<rgba 0 0 0 0>}
       cornerRadii=<4 4 4 4>
      }
      tintColor {colorUseRule=UseColor_Specified, specifiedColor=<rgba 0.0048 0.0048 0.0048 1>}
      uVRegion {bIsValid=false, max=<0 0>, min=<0 0>}
     }
     backgroundImageNormal {
      bIsDynamicallyLoaded=false
      drawAs=RoundedBox
      imageType=NoImage
      mirroring=NoMirror
      resourceName=None
      resourceObject=null
      tiling=NoTile
      imageSize=<0 0>
      margin=<ltrb 0 0 0 0>
      outlineSettings {
       bUseBrushTransparency=false
       roundingType=FixedRadius
       width=1
       color {colorUseRule=UseColor_ColorTable, specifiedColor=<rgba 0 0 0 0>}
       cornerRadii=<4 4 4 4>
      }
      tintColor {colorUseRule=UseColor_Specified, specifiedColor=<rgba 0.0048 0.0048 0.0048 1>}
      uVRegion {bIsValid=false, max=<0 0>, min=<0 0>}
     }
     backgroundImageReadOnly {
      bIsDynamicallyLoaded=false
      drawAs=RoundedBox
      imageType=NoImage
      mirroring=NoMirror
      resourceName=None
      resourceObject=null
      tiling=NoTile
      imageSize=<0 0>
      margin=<ltrb 0 0 0 0>
      outlineSettings {
       bUseBrushTransparency=false
       roundingType=FixedRadius
       width=1
       color {colorUseRule=UseColor_ColorTable, specifiedColor=<rgba 0 0 0 0>}
       cornerRadii=<4 4 4 4>
      }
      tintColor {colorUseRule=UseColor_Specified, specifiedColor=<rgba 0.0284 0.0284 0.0284 1>}
      uVRegion {bIsValid=false, max=<0 0>, min=<0 0>}
     }
     focusedForegroundColor {colorUseRule=UseColor_Specified, specifiedColor=<rgba 1 1 1 1>}
     foregroundColor {colorUseRule=UseColor_Specified, specifiedColor=<rgba 0.5271 0.5271 0.5271 1>}
     hScrollBarPadding=<ltrb 0 0 0 0>
     padding=<ltrb 8 3 8 4>
     readOnlyForegroundColor {colorUseRule=UseColor_Foreground, specifiedColor=<rgba 1 0 1 1>}
     scrollBarStyle {
      thickness=8
      draggedThumbImage {
       bIsDynamicallyLoaded=false
       drawAs=RoundedBox
       imageType=NoImage
       mirroring=NoMirror
       resourceName=None
       resourceObject=null
       tiling=NoTile
       imageSize=<0 0>
       margin=<ltrb 0 0 0 0>
       outlineSettings {
        bUseBrushTransparency=false
        roundingType=FixedRadius
        width=0
        color {colorUseRule=UseColor_Specified, specifiedColor=<rgba 0 0 0 0>}
        cornerRadii=<4 4 4 4>
       }
       tintColor {colorUseRule=UseColor_Specified, specifiedColor=<rgba 0.2159 0.2159 0.2159 1>}
       uVRegion {bIsValid=false, max=<0 0>, min=<0 0>}
      }
      horizontalBackgroundImage {
       bIsDynamicallyLoaded=false
       drawAs=NoDrawType
       imageType=NoImage
       mirroring=NoMirror
       resourceName=None
       resourceObject=null
       tiling=NoTile
       imageSize=<0 0>
       margin=<ltrb 0 0 0 0>
       outlineSettings {
        bUseBrushTransparency=false
        roundingType=HalfHeightRadius
        width=0
        color {colorUseRule=UseColor_Specified, specifiedColor=<rgba 0 0 0 0>}
        cornerRadii=<0 0 0 0>
       }
       tintColor {colorUseRule=UseColor_Specified, specifiedColor=<rgba 1 1 1 1>}
       uVRegion {bIsValid=false, max=<0 0>, min=<0 0>}
      }
      horizontalBottomSlotImage {
       bIsDynamicallyLoaded=false
       drawAs=NoDrawType
       imageType=NoImage
       mirroring=NoMirror
       resourceName=None
       resourceObject=null
       tiling=NoTile
       imageSize=<0 0>
       margin=<ltrb 0 0 0 0>
       outlineSettings {
        bUseBrushTransparency=false
        roundingType=HalfHeightRadius
        width=0
        color {colorUseRule=UseColor_Specified, specifiedColor=<rgba 0 0 0 0>}
        cornerRadii=<0 0 0 0>
       }
       tintColor {colorUseRule=UseColor_Specified, specifiedColor=<rgba 1 1 1 1>}
       uVRegion {bIsValid=false, max=<0 0>, min=<0 0>}
      }
      horizontalTopSlotImage {
       bIsDynamicallyLoaded=false
       drawAs=NoDrawType
       imageType=NoImage
       mirroring=NoMirror
       resourceName=None
       resourceObject=null
       tiling=NoTile
       imageSize=<0 0>
       margin=<ltrb 0 0 0 0>
       outlineSettings {
        bUseBrushTransparency=false
        roundingType=HalfHeightRadius
        width=0
        color {colorUseRule=UseColor_Specified, specifiedColor=<rgba 0 0 0 0>}
        cornerRadii=<0 0 0 0>
       }
       tintColor {colorUseRule=UseColor_Specified, specifiedColor=<rgba 1 1 1 1>}
       uVRegion {bIsValid=false, max=<0 0>, min=<0 0>}
      }
      hoveredThumbImage {
       bIsDynamicallyLoaded=false
       drawAs=RoundedBox
       imageType=NoImage
       mirroring=NoMirror
       resourceName=None
       resourceObject=null
       tiling=NoTile
       imageSize=<0 0>
       margin=<ltrb 0 0 0 0>
       outlineSettings {
        bUseBrushTransparency=false
        roundingType=FixedRadius
        width=0
        color {colorUseRule=UseColor_Specified, specifiedColor=<rgba 0 0 0 0>}
        cornerRadii=<4 4 4 4>
       }
       tintColor {colorUseRule=UseColor_Specified, specifiedColor=<rgba 0.2159 0.2159 0.2159 1>}
       uVRegion {bIsValid=false, max=<0 0>, min=<0 0>}
      }
      normalThumbImage {
       bIsDynamicallyLoaded=false
       drawAs=RoundedBox
       imageType=NoImage
       mirroring=NoMirror
       resourceName=None
       resourceObject=null
       tiling=NoTile
       imageSize=<0 0>
       margin=<ltrb 0 0 0 0>
       outlineSettings {
        bUseBrushTransparency=false
        roundingType=FixedRadius
        width=0
        color {colorUseRule=UseColor_Specified, specifiedColor=<rgba 0 0 0 0>}
        cornerRadii=<4 4 4 4>
       }
       tintColor {colorUseRule=UseColor_Specified, specifiedColor=<rgba 0.0953 0.0953 0.0953 1>}
       uVRegion {bIsValid=false, max=<0 0>, min=<0 0>}
      }
      verticalBackgroundImage {
       bIsDynamicallyLoaded=false
       drawAs=NoDrawType
       imageType=NoImage
       mirroring=NoMirror
       resourceName=None
       resourceObject=null
       tiling=NoTile
       imageSize=<0 0>
       margin=<ltrb 0 0 0 0>
       outlineSettings {
        bUseBrushTransparency=false
        roundingType=HalfHeightRadius
        width=0
        color {colorUseRule=UseColor_Specified, specifiedColor=<rgba 0 0 0 0>}
        cornerRadii=<0 0 0 0>
       }
       tintColor {colorUseRule=UseColor_Specified, specifiedColor=<rgba 1 1 1 1>}
       uVRegion {bIsValid=false, max=<0 0>, min=<0 0>}
      }
      verticalBottomSlotImage {
       bIsDynamicallyLoaded=false
       drawAs=NoDrawType
       imageType=NoImage
       mirroring=NoMirror
       resourceName=None
       resourceObject=null
       tiling=NoTile
       imageSize=<0 0>
       margin=<ltrb 0 0 0 0>
       outlineSettings {
        bUseBrushTransparency=false
        roundingType=HalfHeightRadius
        width=0
        color {colorUseRule=UseColor_Specified, specifiedColor=<rgba 0 0 0 0>}
        cornerRadii=<0 0 0 0>
       }
       tintColor {colorUseRule=UseColor_Specified, specifiedColor=<rgba 1 1 1 1>}
       uVRegion {bIsValid=false, max=<0 0>, min=<0 0>}
      }
      verticalTopSlotImage {
       bIsDynamicallyLoaded=false
       drawAs=NoDrawType
       imageType=NoImage
       mirroring=NoMirror
       resourceName=None
       resourceObject=null
       tiling=NoTile
       imageSize=<0 0>
       margin=<ltrb 0 0 0 0>
       outlineSettings {
        bUseBrushTransparency=false
        roundingType=HalfHeightRadius
        width=0
        color {colorUseRule=UseColor_Specified, specifiedColor=<rgba 0 0 0 0>}
        cornerRadii=<0 0 0 0>
       }
       tintColor {colorUseRule=UseColor_Specified, specifiedColor=<rgba 1 1 1 1>}
       uVRegion {bIsValid=false, max=<0 0>, min=<0 0>}
      }
     }
     textStyle {
      overflowPolicy=Clip
      transformPolicy=None
      colorAndOpacity {colorUseRule=UseColor_Foreground, specifiedColor=<rgba 1 0 1 1>}
      font {
       bForceMonospaced=false
       bMaterialIsStencil=false
       fontMaterial=null
       fontObject=null
       letterSpacing=0
       monospacedWidth=1
       size=10
       skewAmount=0
       typefaceFontName=Regular
       outlineSettings {
        bApplyOutlineToDropShadows=false
        bMiteredCorners=false
        bSeparateFillAlpha=false
        outlineMaterial=null
        outlineSize=0
        outlineColor=<rgba 0 0 0 1>
       }
      }
      highlightColor {colorUseRule=UseColor_ColorTable, specifiedColor=<rgba 0 0 0 0>}
      highlightShape {
       bIsDynamicallyLoaded=false
       drawAs=Image
       imageType=NoImage
       mirroring=NoMirror
       resourceName=None
       resourceObject=null
       tiling=NoTile
       imageSize=<0 0>
       margin=<ltrb 0 0 0 0>
       outlineSettings {
        bUseBrushTransparency=false
        roundingType=HalfHeightRadius
        width=0
        color {colorUseRule=UseColor_Specified, specifiedColor=<rgba 0 0 0 0>}
        cornerRadii=<0 0 0 0>
       }
       tintColor {colorUseRule=UseColor_ColorTable, specifiedColor=<rgba 0 0 0 0>}
       uVRegion {bIsValid=false, max=<0 0>, min=<0 0>}
      }
      selectedBackgroundColor {colorUseRule=UseColor_ColorTable, specifiedColor=<rgba 0 0 0 0>}
      shadowColorAndOpacity=<rgba 0 0 0 1>
      shadowOffset=<0 0>
      strikeBrush {
       bIsDynamicallyLoaded=false
       drawAs=Image
       imageType=NoImage
       mirroring=NoMirror
       resourceName=None
       resourceObject=null
       tiling=NoTile
       imageSize=<32 32>
       margin=<ltrb 0 0 0 0>
       outlineSettings {
        bUseBrushTransparency=false
        roundingType=HalfHeightRadius
        width=0
        color {colorUseRule=UseColor_Specified, specifiedColor=<rgba 0 0 0 0>}
        cornerRadii=<0 0 0 0>
       }
       tintColor {colorUseRule=UseColor_Specified, specifiedColor=<rgba 1 1 1 1>}
       uVRegion {bIsValid=false, max=<0 0>, min=<0 0>}
      }
      underlineBrush {
       bIsDynamicallyLoaded=false
       drawAs=Image
       imageType=NoImage
       mirroring=NoMirror
       resourceName=None
       resourceObject=null
       tiling=NoTile
       imageSize=<32 32>
       margin=<ltrb 0 0 0 0>
       outlineSettings {
        bUseBrushTransparency=false
        roundingType=HalfHeightRadius
        width=0
        color {colorUseRule=UseColor_Specified, specifiedColor=<rgba 0 0 0 0>}
        cornerRadii=<0 0 0 0>
       }
       tintColor {colorUseRule=UseColor_Specified, specifiedColor=<rgba 1 1 1 1>}
       uVRegion {bIsValid=false, max=<0 0>, min=<0 0>}
      }
     }
     vScrollBarPadding=<ltrb 0 0 0 0>
    }}
   slot {class=HorizontalBoxSlot, properties {size {sizeRule=Fill, value=1}}}
  }
  TL_FilteredList {
   widgetClassPath=/Game/Blueprints/Blutilities/ForTechArt/Widgets/List/TA_Generic01_List
   bIsVariable=true
   parent=VerticalBox_1
   widget=TL_FilteredList
   slot {class=VerticalBoxSlot, properties {size {sizeRule=Fill, value=1}}}
  }
  BTN_EditWidget {
   widgetClassPath=&TA_Generic01_Button
   parent=HorizontalBox_0
   widget=BTN_EditWidget
   properties {label=">>", buttonColor_Default=<rgba 0.0469 0.0469 0.0469 1>, sizeOverride=<20 0>}
   slot {class=HorizontalBoxSlot, properties {padding=<ltrb 5 5 5 5>}}
  }
 ]
}
